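# Sound cue that plays the drop sound for the Squash item in HighTower, classed as loot collect audio.
# A random node picks one of two drop waves at equal weight and feeds a modulator at the root for pitch and volume variation.
# It plays at reduced volume with weapon-drop attenuation, and its concurrency override prevents new instances once the limit is reached.
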SoundCue /HighTower/Audio/Squash/Hightower_Squash_Drop_Cue
root=Modulator_0
nodes [
 {id=Modulator_0, children [Random_1]}
 {id=Random_1, children [WavePlayer_1,WavePlayer_3], properties {weights [1,1]}}
 {id=WavePlayer_1, properties {soundWaveAssetPtr=&/Hightower_Squash_Drop_02}}
 {id=WavePlayer_3, properties {soundWaveAssetPtr=&/Hightower_Squash_Drop_01}}
]
properties {
 attenuationSettings=&Fort_Weapon_Drop_Att
 bOverrideConcurrency=true
 maxDistance=3000
 soundClassObject=&Fort_Loot_Collect
 volumeMultiplier=0.5
 concurrencyOverrides {resolutionRule=PreventNew}
}
