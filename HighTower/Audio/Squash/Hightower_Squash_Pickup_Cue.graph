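# Sound cue for picking up the Hightower Squash item, classed as a loot-collect sound.
# A Random node picks one of two pickup waves with equal weight and feeds a Modulator at the root.
# It plays quietly with weapon-pickup attenuation and overrides concurrency so new instances are prevented rather than cutting off playing ones.
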SoundCue /HighTower/Audio/Squash/Hightower_Squash_Pickup_Cue
root=Modulator_0
nodes [
 {id=Modulator_0, children [Random_0]}
 {id=Random_0, children [WavePlayer_0,WavePlayer_2], properties {weights [1,1]}}
 {id=WavePlayer_0, properties {soundWaveAssetPtr=&/Hightower_Squash_PickUp_02}}
 {id=WavePlayer_2, properties {soundWaveAssetPtr=&/Hightower_Squash_PickUp_01}}
]
properties {
 attenuationSettings=&Fort_Weapon_Pickup_Att
 bOverrideConcurrency=true
 maxDistance=3500
 soundClassObject=&Fort_Loot_Collect
 volumeMultiplier=0.15
 concurrencyOverrides {resolutionRule=PreventNew}
}
