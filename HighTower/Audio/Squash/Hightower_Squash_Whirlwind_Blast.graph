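# Sound cue for the Whirlwind Blast effect in HighTower's Squash audio set, classed as third-person ranged weapon fire.
# It picks one of three equally weighted wave variants, then applies slight pitch and volume modulation.
# A borrowed Milo close-explosion attenuation spatializes it, with the falloff distance overridden.
# A concurrency override allows one instance per owner, releasing the oldest when a new one starts.
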SoundCue /HighTower/Audio/Squash/Hightower_Squash_Whirlwind_Blast
root=Attenuation_7
nodes [
 {
  id=Attenuation_7
  children [Modulator_6]
  properties {attenuationSettings=/Game/Sounds/Attenuation/Fort_Weapon_Milo_Close_Explosion_Athena_Att, attenuationOverrides {falloffDistance=1600}}
 }
 {id=Modulator_6, children [Random_0], properties {pitchMin=1, volumeMin=0.8}}
 {id=Random_0, children [WavePlayer_9, WavePlayer_10, WavePlayer_11], properties {weights [1,1,1]}}
 {id=WavePlayer_10, properties {soundWaveAssetPtr=&/Hightower_Squash_Whirlwind_Blast_02}}
 {id=WavePlayer_11, properties {soundWaveAssetPtr=&/Hightower_Squash_Whirlwind_Blast_03}}
 {id=WavePlayer_9, properties {soundWaveAssetPtr=&/Hightower_Squash_Whirlwind_Blast_01}}
]
properties {
 bOverrideConcurrency=true
 maxDistance=7500
 priority=2
 soundClassObject=&Fort_WeaponRanged_Fire_3P_PSM
 volumeMultiplier=0.8
 concurrencyOverrides {bLimitToOwner=true, bVolumeScaleCanRelease=true, maxCount=1, resolutionRule=StopOldest, voiceStealReleaseTime=1}
 concurrencySet [null]
}
